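# AssetValidationHelpers is an editor utility Blueprint that locks and unlocks selected assets against edits during development.
# LockAssetSelection and UnlockAssetSelection each fetch the selected assets and pass them as an input to an inline Execute Python Script node.
# The scripts call LockAssetActionsLibrary from the protected_assets Python module, so the actual protection logic lives outside this Blueprint.
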EditorUtilityBlueprint /Game/NeverCook/TechArt/AssetValidationHelpers
graphs [
 EventGraph {links [], nodes []}
 LockAssetSelection {nodes [
   LockAssetSelection() {is=title K2Node_FunctionEntry}
   > n1 = GetSelectedAssets()
   > Execute Python Script!(
    execute
    PythonScript="from protected_assets import LockAssetActionsLibrary\r\n\r\nLockAssetActionsLibrary.lock_assets_from_edits(assets)"
    assets:Array of Object References=@n1
   ) -> (then, ReturnValue) {inputs [assets]}
  ]}
 UnlockAssetSelection {nodes [
   UnlockAssetSelection() {is=title K2Node_FunctionEntry}
   > n1 = GetSelectedAssets()
   > Execute Python Script(PythonScript="from protected_assets import LockAssetActionsLibrary\r\n\r\nLockAssetActionsLibrary.unlock_assets(assets)", assets=@n1) {inputs [assets]}
  ]}
]
variables []
