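# Unlit debug material that shows a UV checker for judging texel density and UV layout on meshes.
# A densely tiled CheckerPattern function drives a lerp between dimmed and full brightness.
# That checker multiplies the UV coordinates appended with zero, so U reads as red and V as green beneath the squares.
# Usage is fixed to Nanite, skeletal mesh and mesh deformer, with automatic usage setting turned off.
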Material /Game/NeverCook/TechArt/M_UVDebug_TexelDensity_Checker
graph {
 properties {
  bAutomaticallySetUsageInEditor=false
  bUsedWithMeshDeformer=true
  bUsedWithNanite=true
  bUsedWithSkeletalMesh=true
  shadingModel=MSM_Unlit
  usedShadingModels=Unlit
 }
 nodes [
  n1 = TextureCoordinate()
  n2 = Constant() {r=64}
  Multiply(
   A=LinearInterpolate(Alpha=MaterialFunctionCall!(UVs=@n1, Tiling=@n2) {materialFunction=/Engine/ArtTools/RenderToTexture/MaterialFunctions/CheckerPattern}.Result) {constA=0.25}
   B=AppendVector(A=@n1, B=Constant())
  )
 ]
}
